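# UI-domain material for PastaSauce-themed Item Shop tiles, building a pink, crimson and purple backdrop with scanlines.
# Offset radial gradients blend the colours, with a top-bottom gradient darkening toward black.
# A panned cracks texture, controlled by ScanlineSpeed and ScanlineOpacity, forms moving scanlines.
# Repeating UV lines combine with an aspect-corrected box edge mask into a scanline alpha, shared through named reroutes.
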Material /ItemShop/UI/Art/PastaSauce/M_UI_PastaSauce_ShopTiles
graph {
 properties {bUsedWithStaticMesh=false, materialDomain=MD_UI, pixelDepthOffsetMode=PDOM_Legacy, usedShadingModels=Default Lit}
 nodes [
  n1 = Constant3Vector() {constant=<rgba 1 0.6376 0.8388 1>}
  n2 = NamedRerouteUsage()
  n3 = AppendVector(A=Constant(), B=Constant())
  n4 = SmoothStep(Value=MaterialFunctionCall(UV=Add(A=@n2, B=@n3), GradientStart=Constant() {r=-0.25}, GradientEnd=Constant() {r=1}) {materialFunction=&MF_UI_RadialGradient}.Smooth) {constMax=1.5}
  n5 = LinearInterpolate(A=Constant3Vector() {constant=<rgba 0.6308 0.0602 0.1695 1>}, B=@n1, Alpha=@n4)
  n6 = Constant3Vector() {constant=<rgba 0.3177 0 0.1717 1>}
  n7 = NamedRerouteUsage()
  n8 = Add(A=NamedRerouteUsage(), B=AppendVector(A=Constant(), B=Constant()))
  n9 = Constant() {r=-0.5}
  n10 = Constant() {r=1}
  n11 = Multiply(A=SmoothStep(Value=MaterialFunctionCall(UV=@n8, GradientStart=@n9, GradientEnd=@n10) {materialFunction=&MF_UI_RadialGradient}.Smooth) {constMax=1.25}) {constB=2}
  n12 = Saturate(None=Subtract(A=Saturate(None=@n7), B=Power(Base=@n11) {constExponent=0.25}))
  n13 = MaterialFunctionCall!(UVs=NamedRerouteUsage(), Subtract=Constant() {r=0.2}, Falloff=Constant() {r=1}, Intensity=Constant() {r=1.25}) {materialFunction=/Game/UI/Frontend/ItemShop/Art/_MaterialFunctions_ShopTheming/MF_UI_TopBottomGradient}
  LinearInterpolate(
   A=LinearInterpolate(
    A=LinearInterpolate(A=@n5, B=@n6, Alpha=@n12)
    B=Constant3Vector() {constant=<rgba 0 0 0 1>}
    Alpha=SmoothStep(Value=@n13.Result) {constMax=1.2}
   ) {constB=0}
   B=@n1
   Alpha=@n4
  )
  Multiply(A=@n7) {constB=0.2}
  n14 = Constant2Vector() {g=0.65, r=0.25}
  n15 = TextureCoordinate() {coordinateIndex=1}
  Multiply(
   A=@n14
   B=ComponentMask(None=MaterialFunctionCall!(UV=@n15, Seed) {materialFunction=&MF_UI_Random}.Result) {g=true}
  )
  n16 = Time()
  n17 = ConstantBiasScale(None=Sine(None=Multiply(A=@n16) {constB=0}))
  n18 = TextureCoordinate() {coordinateIndex=3}
  n19 = ComponentMask(None=@n18) {r=true}
  n20 = ComponentMask(None=@n18) {g=true}
  n21 = Constant() {r=0.7}
  n22 = Multiply(A=OneMinus(None=Divide(A=@n19, B=@n20)), B=@n21)
  n23 = AppendVector(A=@n21, B=Add(A=Multiply(A=@n22) {constB=0.5}, B=@n21))
  n24 = Constant()
  n25 = Constant() {r=0.65}
  n26 = StaticBool() {value=true}
  n27 = OneMinus(None=SmoothStep(
    Max=LinearInterpolate(Alpha=@n17) {constA=0.8}
    Value=MaterialFunctionCall(Size=@n23, CornerRoundness=@n24, GlowMax=@n25, KeepAspectRatio=@n26) {materialFunction=&MF_UI_Box}.Fill
   ))
  Multiply(A=@n27, B=Constant3Vector() {constant=<rgba 1 0 0 0>})
  n28 = Constant()
  n29 = Multiply(A=ComponentMask(None=@n15) {g=true}) {constB=3}
  n30 = ScalarParameter() {defaultValue=0.12, parameterName=ScanlineSpeed}
  n31 = TextureSample(UVs=Panner(Coordinate=AppendVector(A=@n28, B=@n29), Speed=Multiply(A=@n14, B=@n30) {constB=0.2}) {speedX=0.257, speedY=0.651}) {samplerType=SAMPLERTYPE_LinearColor, texture=/Game/Global/Textures/Default/Masks_and_Gradients/T_Cracks}
  n32 = Multiply(A=@n31.R, B=@n31.R)
  n33 = Multiply(A=@n32, B=@n32)
  n34 = ScalarParameter() {defaultValue=0.25, parameterName=ScanlineOpacity}
  NamedRerouteDeclaration(None=Saturate(None=Multiply(A=@n33, B=@n34) {constB=0.25})) {name=MovingScanlines, nodeColor=<rgba 0 0.1562 1 1>}
  NamedRerouteDeclaration(None=MaterialFunctionCall() {materialFunction=&MF_UI_UVScaling}.Scaled To Fill) {name=UVFill_1, nodeColor=<rgba 0.9922 0 1 1>}
  n35 = NamedRerouteUsage()
  n36 = Multiply(A=MaterialFunctionCall(Base=Multiply(A=@n27) {constB=0.1}, Blend=@n35) {materialFunction=&Blend_Overlay}.Result) {constB=100}
  n37 = TextureCoordinate()
  n38 = Frac(None=Multiply(A=ComponentMask(None=@n37) {g=true}) {constB=42.5, desc=Number of lines})
  n39 = Add(A=Floor(None=Divide(A=@n38) {constB=0.66, desc="empty space sizing (percentage)"}), B=Multiply(A=@n35) {constB=3})
  NamedRerouteDeclaration(None=Saturate(None=Add(A=@n36, B=@n39))) {name=ScanlinesAlpha, nodeColor=<rgba 1 0.8672 0 1>}
  TextureCoordinate() {coordinateIndex=3}
 ]
}
