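# Animated background base material for the Pilgrim event screen in the item shop, with a floor setup function and IsFloor switch.
# Pulsing radial gradients, cycling cosine palettes, bokeh and soft-light blending build the colour.
# A rotated, screen-aligned polar UV warp drives two panned grids of randomly spun, scaled shapes, masked by screen gradients.
# The "Triagles" layer draws MF_UI_Box shapes, as both triangle function calls have unused outputs.
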Material /ItemShop/UI/Art/PastaSauce/M_UI_PilgrimEventScreen_ShopBGBase
graph {
 properties {pixelDepthOffsetMode=PDOM_Legacy, usedShadingModels=Default Lit}
 nodes [
  n1 = MaterialFunctionCall() {materialFunction=&ScreenAlignedUVs}
  n2 = MaterialFunctionCall(Float2=MaterialFunctionCall() {materialFunction=&ScreenResolution}.Visible Resolution) {materialFunction=&BreakOutFloat2Components}
  n3 = AppendVector(A=Divide(A=@n2.R, B=@n2.R), B=Constant() {r=1})
  n4 = ScalarParameter() {defaultValue=-0.596, parameterName=X-Offset}
  n5 = ScalarParameter() {parameterName=Y-Offset}
  n6 = Rotator(Coordinate=Add(A=Multiply(A=@n1."X 100%, Y 100%", B=@n3), B=AppendVector(A=@n4, B=@n5))) {speed=0.069}
  n7 = Subtract(A=@n6) {constB=0.5}
  n8 = Multiply(A=@n7) {constB=-1}
  n9 = Ceil(None=ComponentMask(None=Add(A=@n6) {constB=-0.5}) {g=true})
  n10 = NamedRerouteDeclaration(None=LinearInterpolate(
    A=MaterialFunctionCall(Vector or UVs=@n7) {materialFunction=&VectorToRadialValue}.Radial Coordinates
    B=MaterialFunctionCall(Vector or UVs=@n8) {materialFunction=&VectorToRadialValue}.Radial Coordinates
    Alpha=Saturate(None=@n9)
   )) {name=UVMask, nodeColor=<rgba 0 0 0 1>}
  n11 = ComponentMask(None=@n10) {r=true}
  n12 = ComponentMask(None=@n10) {g=true}
  n13 = Multiply(
   A=MaterialFunctionCall(X=Divide(A=@n11, B=@n12), Y=@n11) {materialFunction=&MakeFloat2}.Result
   B=Constant2Vector() {g=2, r=0.25}
  ) {constB=3}
  n14 = Panner(Coordinate=@n13, Time=Multiply(A=Time(), B=ScalarParameter() {defaultValue=0.55, parameterName=Time-Multiply-Triangles-b})) {speedX=0.069, speedY=0.013}
  n15 = Constant() {r=-0.125}
  n16 = Multiply(A=Add(A=Rotator(Coordinate=@n14, Time=@n15) {speed=6.2832}, B=Constant2Vector() {g=0.125, r=0.125})) {constB=10}
  n17 = Frac(None=@n16)
  n18 = MaterialFunctionCall!(
   PrimaryTextureCoordinates=@n16
   2ndOrderTextureCoordinates
   RandomSeed=Constant() {r=98765}
   NoiseDensityBias
   Bypass Noise Influence Crosstalk
  ) {materialFunction=/Game/Effects/Fort_Effects/Materials/_Functions/MF_RandomUVSample}
  n19 = Multiply(A=LinearInterpolate(Alpha=@n18.NoiseWithRange) {constA=-0.25, constB=0.25}, B=Time())
  n20 = MaterialFunctionCall(
   UVs=Add(A=Rotator(Coordinate=@n17, Time=@n19) {speed=6.2832}, B=LinearInterpolate(Alpha=@n18.NoiseWithRange) {constA=-0.35, constB=0})
   Texture Scale=LinearInterpolate(Alpha=@n18.Noise) {constA=0.125, constB=0.5}
  ) {materialFunction=&ScaleUVsByCenter}
  MaterialFunctionCall!(UVs=@n20.UVs, Corner 01, Corner 02, Corner 03, Glow Min, Glow Max, Outline Thickness) {materialFunction=/FortUILibrary/MaterialFunctions/Shapes/MF_UI_WIP_Triangle}
  n21 = Time()
  n22 = ScalarParameter() {defaultValue=0.35, parameterName=Time-Multiply-Triangles-a}
  n23 = Multiply(A=Panner(Coordinate=@n13, Time=Multiply(A=@n21, B=@n22)) {speedX=0.038, speedY=0.013})
  n24 = Constant() {r=0.125}
  n25 = Multiply(A=Add(A=Rotator(Coordinate=@n23, Time=@n24) {speed=6.2832}, B=Constant2Vector() {g=0.25, r=0.35}) {constB=0.25}) {constB=5}
  n26 = Frac(None=@n25)
  n27 = MaterialFunctionCall(PrimaryTextureCoordinates=@n25, RandomSeed=Constant() {r=1234}) {materialFunction=&MF_RandomUVSample}
  n28 = Multiply(A=LinearInterpolate(Alpha=@n27.NoiseWithRange) {constA=-0.125, constB=0.125}, B=Time())
  n29 = MaterialFunctionCall(
   UVs=Add(A=Rotator(Coordinate=@n26, Time=@n28) {speed=6.2832}, B=LinearInterpolate(Alpha=@n27.NoiseWithRange) {constA=-0.125, constB=0.125})
   Texture Scale=LinearInterpolate(Alpha=@n27.Noise) {constA=0.125, constB=0.5}
  ) {materialFunction=&ScaleUVsByCenter}
  MaterialFunctionCall(UVs=@n29.UVs) {materialFunction=&MF_UI_WIP_Triangle}
  MaterialFunctionCall!(
   Color=NamedRerouteUsage()
   Color Intensity=ScalarParameter() {defaultValue=1, group=0 - Global, parameterName=Color Intensity}
   Floor Radius=ScalarParameter() {defaultValue=0.175, group=6 - Floor, parameterName=Floor Radius}
   Floor Density=ScalarParameter() {defaultValue=1.75, group=6 - Floor, parameterName=Floor Density}
   Floor Metallic=ScalarParameter() {group=6 - Floor, parameterName=Floor Metalic}
   Floor Specular=ScalarParameter() {group=6 - Floor, parameterName=Floor Spec}
   Floor Roughness=ScalarParameter() {defaultValue=1, group=6 - Floor, parameterName=Floor Roughness}
   IsFloor=StaticBoolParameter() {group=6 - Floor, parameterName=IsFloor}
  ) {materialFunction=/Game/UI/Foundation/Environment/Shell/Materials/MaterialFunctions/MF_UI_BP_FloorSetup}
  n30 = Add(A=NamedRerouteUsage(), B=ScalarParameter() {defaultValue=0.05, parameterName=Hotspot Displacement})
  n31 = NamedRerouteUsage()
  n32 = LinearInterpolate(
   A=ScalarParameter() {defaultValue=2.5, group=3 - Gradient Overlay, parameterName=Gradient C Power Min}
   B=ScalarParameter() {defaultValue=3, group=3 - Gradient Overlay, parameterName=Gradient C Power Max}
   Alpha=Divide(A=Multiply(A=@n31)) {constB=0.4}
  )
  n33 = ScalarParameter() {defaultValue=1.5, group=3 - Gradient Overlay, parameterName="Gradient  Intensity"}
  n34 = NamedRerouteUsage()
  n35 = Add(
   A=NamedRerouteUsage()
   B=AppendVector(A=Multiply(A=@n34) {constB=-1}, B=ScalarParameter() {group=3 - Gradient Overlay, parameterName=Gradient C Offset Y})
  )
  n36 = Constant() {r=1}
  NamedRerouteDeclaration(None=Multiply(
    A=NamedRerouteUsage()
    B=MaterialFunctionCall!(UV=@n30, Gradient Power=@n32, Gradient Intensity=@n33, Gradient Offset=@n35, Time=@n36) {materialFunction=/Game/UI/Foundation/Environment/Shell/Materials/MaterialFunctions/MF_UI_BP_PulsingRadialGradient}.Result
   )) {name=Central Gradient Color, nodeColor=<rgba 0 0 0 1>}
  n37 = VectorParameter() {group=3 - Gradient Overlay, parameterName=Gradient A Color, defaultValue=<rgba 0.0768 0.0242 0.224 1>}
  n38 = NamedRerouteUsage()
  n39 = NamedRerouteUsage()
  n40 = MaterialFunctionCall!(
   Gradient=ComponentMask(None=Panner(Coordinate=@n39) {speedX=0.2}) {r=true}
   a=VectorParameter() {parameterName=BaseColor, defaultValue=<rgba 0.1169 0.0302 0.3185 1>}.RGB
   b=VectorParameter() {parameterName=BaseColor3, defaultValue=<rgba 0.0307 0.0212 0.2664 1>}.RGB
   c=VectorParameter() {parameterName=BaseColor2, defaultValue=<rgba 0.0187 0.0606 0.3229 1>}.RGB
   d=VectorParameter() {parameterName=BaseColor4, defaultValue=<rgba 0.3005 0.0669 0.2039 1>}.RGB
  ) {materialFunction=/FortUILibrary/MaterialFunctions/Color/MF_UI_CosineGradient}
  n41 = ComponentMask(None=MaterialFunctionCall!(RGB=NamedRerouteDeclaration(None=@n40.Result) {name=RotatingGradient, nodeColor=<rgba 0 0 0 1>}) {materialFunction=/Engine/Functions/Engine_MaterialFunctions02/Math/RGBtoHSV}."RGB to HSV (0-1)") {b=true}
  n42 = NamedRerouteUsage()
  n43 = ScalarParameter() {defaultValue=2, group=3 - Gradient Overlay, parameterName=Gradient A Power}
  n44 = ScalarParameter() {defaultValue=2, group=3 - Gradient Overlay, parameterName=Gradient A Intensity}
  n45 = AppendVector(
   A=ScalarParameter() {defaultValue=0.6, group=3 - Gradient Overlay, parameterName=Gradient A Offset X}
   B=ScalarParameter() {defaultValue=0.75, group=3 - Gradient Overlay, parameterName=Gradient A Offset Y}
  )
  n46 = NamedRerouteUsage()
  NamedRerouteDeclaration(None=Multiply(
    A=LinearInterpolate(A=@n37.RGB, B=@n38, Alpha=@n41)
    B=MaterialFunctionCall(UV=@n42, Gradient Power=@n43, Gradient Intensity=@n44, Gradient Offset=@n45, Time=@n46) {materialFunction=&MF_UI_BP_PulsingRadialGradient}.Result
   )) {name=Top Gradient Color, nodeColor=<rgba 0 0 0 1>}
  n47 = NamedRerouteUsage()
  n48 = ScalarParameter() {defaultValue=2, group=3 - Gradient Overlay, parameterName=Gradient B Power}
  n49 = ScalarParameter() {defaultValue=0.45, group=3 - Gradient Overlay, parameterName=Gradient B Intensity}
  n50 = AppendVector(
   A=ScalarParameter() {defaultValue=0.5, group=3 - Gradient Overlay, parameterName=Gradient B Offset X}
   B=ScalarParameter() {defaultValue=-0.8, group=3 - Gradient Overlay, parameterName=Gradient B Offset Y}
  )
  n51 = NamedRerouteUsage()
  NamedRerouteDeclaration(None=Multiply(
    A=VectorParameter() {group=3 - Gradient Overlay, parameterName=Gradient B Color, defaultValue=<rgba 0.296 0.083 0.5781 1>}.RGB
    B=MaterialFunctionCall(UV=@n47, Gradient Power=@n48, Gradient Intensity=@n49, Gradient Offset=@n50, Time=@n51) {materialFunction=&MF_UI_BP_PulsingRadialGradient}.Result
   )) {name=Bottom Left Gradient Color, nodeColor=<rgba 0 0 0 1>}
  n52 = MaterialFunctionCall() {materialFunction=/Game/UI/Foundation/Environment/Shell/Materials/MaterialFunctions/MF_UI_BP_UvSetup}
  NamedRerouteDeclaration(None=@n52.Squared UV) {name=Squared UV, nodeColor=<rgba 0 0 0 1>}
  n53 = MaterialFunctionCall() {materialFunction=&ScreenAlignedUVs}
  n54 = OneMinus(None=@n53."X 100%, Y 100%")
  n55 = Multiply(A=Add(A=ComponentMask(None=@n54) {g=true}) {constB=0.25}) {constB=1.5}
  n56 = Add(A=ComponentMask(None=@n53."X 100%, Y 100%") {g=true}) {constB=0.25}
  NamedRerouteDeclaration(None=Multiply(A=Saturate(None=@n55), B=Saturate(None=@n56))) {name=TopGradient_1, nodeColor=<rgba 0 0 0 1>}
  n57 = Add(A=@n53."X 100%, Y 100%") {constB=-0.5}
  n58 = OneMinus(None=ComponentMask(None=Abs(None=@n57)) {r=true})
  NamedRerouteDeclaration(None=Saturate(None=Add(A=@n58) {constB=0.125})) {name=MidGradient_1, nodeColor=<rgba 0 0 0 1>}
  n59 = Panner(Coordinate=NamedRerouteUsage()) {speedX=-0.1}
  NamedRerouteDeclaration(None=MaterialFunctionCall(
    Gradient=ComponentMask(None=@n59) {r=true}
    a=VectorParameter() {group=3 - Gradient Overlay, parameterName=Gradient A Color, defaultValue=<rgba 0.0768 0.0242 0.224 1>}.RGB
    b=VectorParameter() {group=3 - Gradient Overlay, parameterName=Gradient B Color, defaultValue=<rgba 0.296 0.083 0.5781 1>}.RGB
    c=VectorParameter() {parameterName=BaseColor4, defaultValue=<rgba 0.013 0.0423 0.3005 1>}.RGB
    d=VectorParameter() {group=3 - Gradient Overlay, parameterName=Gradient C Color, defaultValue=<rgba 0 0.55 1 1>}.RGB
   ) {materialFunction=&MF_UI_CosineGradient}.Result) {name=RotatingGradient2, nodeColor=<rgba 0 0 0 1>}
  NamedRerouteDeclaration(None=Time()) {name=Time, nodeColor=<rgba 0 0 0 1>}
  NamedRerouteDeclaration(None=@n52.Squared UV Center Offset) {name=Squared UV Center Offset, nodeColor=<rgba 0 0 0 1>}
  NamedRerouteDeclaration(None=ScalarParameter() {defaultValue=0.5, group=Glow-Centre, parameterName=Cnt_Glow_x-Offset}) {name=Hotspot X-Offset, nodeColor=<rgba 0 0 0 1>}
  NamedRerouteDeclaration(None=ScalarParameter() {defaultValue=1, group=fx-streaks, parameterName=bIsUseStreak}) {name=FX Vis Control, nodeColor=<rgba 0 0 0 1>}
  n60 = NamedRerouteUsage()
  n61 = NamedRerouteUsage()
  n62 = VectorParameter() {group=1 - Chrome Visual, parameterName=Color C, sortPriority=3, defaultValue=<rgba 0.0123 0.0027 0.0578 1>}
  n63 = Desaturation(None=NamedRerouteUsage(), Fraction=Constant() {r=0.5})
  n64 = VectorParameter() {group=1 - Chrome Visual, parameterName=Color B, sortPriority=2, defaultValue=<rgba 0 0.0429 0.6979 1>}
  n65 = Power(
   Base=Desaturation(None=NamedRerouteUsage(), Fraction=Constant() {r=1})
   Exp=ScalarParameter() {defaultValue=0.4, parameterName=Rear Gradient Power}
  )
  n66 = NamedRerouteUsage()
  n67 = Constant() {r=1}
  n68 = Power(
   Base=Clamp(None=Desaturation(None=@n66, Fraction=@n67), Max=ScalarParameter() {defaultValue=0.6, parameterName=Max Clamp})
   Exp=ScalarParameter() {defaultValue=1.2, parameterName=Hotspot Gradient Power}
  )
  n69 = Add(
   A=LinearInterpolate(A=@n62.RGB, B=Add(A=@n63, B=@n64.RGB), Alpha=Saturate(None=@n65))
   B=LinearInterpolate(
    B=VectorParameter() {group=3 - Gradient Overlay, parameterName=Gradient C Color, defaultValue=<rgba 0 0.55 1 1>}.RGB
    Alpha=Saturate(None=@n68)
   )
  )
  n70 = NamedRerouteUsage()
  n71 = MaterialFunctionCall!(Base=Add(A=Add(A=@n69, B=@n70), B=NamedRerouteUsage()), Blend=NamedRerouteUsage()) {materialFunction=/Engine/Functions/Engine_MaterialFunctions03/Blends/Blend_SoftLight}
  n72 = Multiply(A=NamedRerouteUsage(), B=VectorParameter() {group=3 - Gradient Overlay, parameterName=Gradient C Color, defaultValue=<rgba 0 0.55 1 1>}.RGB)
  NamedRerouteDeclaration(None=Multiply(A=Multiply(A=@n60, B=@n61), B=Add(A=@n71.Result, B=@n72))) {name=Final Color, nodeColor=<rgba 0 0 0 1>}
  NamedRerouteDeclaration(None=MaterialFunctionCall!(
    UV=NamedRerouteUsage()
    Bokeh Texture
    Bokeh Color=VectorParameter() {group=4 - Bokeh, parameterName=Bokeh Color, defaultValue=<rgba 1 1 1 1>}.RGB
    Bokeh Opacity=ScalarParameter() {defaultValue=0.0312, group=4 - Bokeh, parameterName=Bokeh Opacity}
    Bokeh Power
    Bokeh Scale
    Foreground Scale
    Background Scale
    Bokeh Angle
    Bokeh Speed
    Falloff Intensity
    Falloff Edge
   ) {materialFunction=/Game/UI/Foundation/Environment/Shell/Materials/MaterialFunctions/MF_UI_BP_Bokeh}.Result) {name=Bokeh, nodeColor=<rgba 0 0 0 1>}
  n73 = Multiply(A=NamedRerouteUsage(), B=NamedRerouteUsage())
  n74 = Power(
   Base=Multiply(A=OneMinus(None=@n73), B=ScalarParameter() {defaultValue=1.75, parameterName=Mask-Intensity}) {constB=4}
   Exp=ScalarParameter() {defaultValue=6, parameterName=Mask-Power}
  )
  n75 = OneMinus(None=MaterialFunctionCall() {materialFunction=&ScreenAlignedUVs}."X 100%, Y 100%")
  n76 = Saturate(None=Subtract(A=ComponentMask(None=@n75) {g=true}, B=ScalarParameter() {defaultValue=0.25, parameterName=Gradient Offset}))
  n77 = ComponentMask(None=NamedRerouteUsage()) {g=true}
  n78 = Multiply(A=@n27.NoiseWithRange, B=Saturate(None=OneMinus(None=@n77)))
  n79 = LinearInterpolate(Alpha=@n77) {constA=0.02, constB=0.16}
  n80 = MaterialFunctionCall(UVs=@n29.UVs) {materialFunction=&MF_UI_Box}
  n81 = OneMinus(None=Clamp(None=@n80.SDFOutline))
  n82 = Add(
   A=LinearInterpolate(Alpha=SmoothStep(Max=@n79, Value=@n80.SDFOutline) {constMax=0.02}) {constA=1, constB=0}
   B=Multiply(A=Subtract(A=@n81) {constB=0.85}) {constB=0.5}
  )
  n83 = ComponentMask(None=NamedRerouteUsage()) {g=true}
  n84 = Multiply(A=@n18.NoiseWithRange, B=Saturate(None=OneMinus(None=@n83)))
  n85 = LinearInterpolate(Alpha=@n83) {constA=0.02, constB=0.16}
  n86 = MaterialFunctionCall(UVs=@n20.UVs) {materialFunction=&MF_UI_Box}
  n87 = OneMinus(None=Clamp(None=@n86.SDFOutline))
  n88 = Add(
   A=LinearInterpolate(Alpha=SmoothStep(Max=@n85, Value=@n86.SDFOutline) {constMax=0.02}) {constA=1, constB=0}
   B=Multiply(A=Subtract(A=@n87) {constB=0.85}) {constB=0.5}
  )
  n89 = Max(A=Multiply(A=Saturate(None=@n78), B=Saturate(None=@n82)), B=Multiply(A=Saturate(None=@n84), B=Saturate(None=@n88)))
  NamedRerouteDeclaration(None=Multiply(A=Saturate(None=@n74), B=Multiply(A=@n76, B=@n89))) {name=Triagles, nodeColor=<rgba 0 0 0 1>}
  NamedRerouteDeclaration(None=@n52.Center Point) {name=Center Point, nodeColor=<rgba 0 0 0 1>}
  n90 = Sine(None=Multiply(A=NamedRerouteUsage(), B=ScalarParameter() {defaultValue=0.125, group=3 - Gradient Overlay, parameterName=Gradient Time Factor}))
  NamedRerouteDeclaration(None=LinearInterpolate(Alpha=ConstantBiasScale(None=@n90)) {constA=0.5}) {name=Pulsing Time, nodeColor=<rgba 0 0 0 1>}
  NamedRerouteUsage()
  NamedRerouteUsage()
 ]
}
